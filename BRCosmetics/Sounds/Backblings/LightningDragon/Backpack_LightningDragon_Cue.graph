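# Sound cue for the Lightning Dragon back bling cosmetic.
# A single wave player feeds one Lightning Dragon sound wave straight to the output, with no randomization or modulation.
# It uses close back bling attenuation, the reactive back bling property sheet and the back bling sound class.
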SoundCue /BRCosmetics/Sounds/Backblings/LightningDragon/Backpack_LightningDragon_Cue
root=WavePlayer_0
nodes [{id=WavePlayer_0, properties {soundWaveAssetPtr=&/Backpack_LightningDragon_0101}}]
properties {
 attenuationSettings=/BRCosmetics/Sounds/Other/Attenuation/Fort_BackBling_Close_Att
 audioPropertiesSheet=/BRCosmetics/Sounds/Settings/PropertySheets/Reactive/APS_Cosmetic_Reactive_Backbling
 maxDistance=1111.11
 soundClassObject=/Game/Sounds/FortSoundClassesAndModes/Fort_Backblings
}
